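# Unlit material for the SudsWobbly vehicle cosmetic body, named as a temporary emissive material.
# Its graph holds a single constant colour, an over-bright cyan-blue suited to glow, with no textures, parameters or visible link to an output.
# It is flagged for skeletal mesh use, with automatic usage flagging in the editor turned off.
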Material /VehicleCosmetics/Bodies/SudsWobbly/Materials/M_sudsWobbly_emisive_Temp
graph {
 properties {bAutomaticallySetUsageInEditor=false, bUsedWithSkeletalMesh=true, shadingModel=MSM_Unlit, usedShadingModels=Unlit}
 nodes [Constant4Vector() {constant=<rgba 0 3.2278 10 1>}]
}
