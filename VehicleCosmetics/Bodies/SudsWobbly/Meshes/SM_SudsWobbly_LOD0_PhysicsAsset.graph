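# Physics asset for the LOD0 mesh of the SudsWobbly vehicle body cosmetic.
# It defines no constraints, so it sets up no jointed ragdoll or secondary physics motion between bodies.
# No body setups or collision shapes are shown, leaving it effectively a minimal placeholder at its usual settings.
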PhysicsAsset /VehicleCosmetics/Bodies/SudsWobbly/Meshes/SM_SudsWobbly_LOD0_PhysicsAsset
constraints []
